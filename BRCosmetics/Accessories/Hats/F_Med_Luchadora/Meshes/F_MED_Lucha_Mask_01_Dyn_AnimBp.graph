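# Post-style Anim Blueprint for the female medium luchadora mask, adding secondary motion to its dyn_hair bones on the Fortnite base head skeleton.
# The input pose passes through two Anim Dynamics nodes for dyn_hair_01 and dyn_hair_02, using angular springs, tight limits, low alpha and no wind.
# A separate four-node chain with linear springs, head-relative limits and tip-bone simulation is disconnected from the output.
# The event graph has no logic.
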AnimBlueprint /BRCosmetics/Accessories/Hats/F_Med_Luchadora/Meshes/F_MED_Lucha_Mask_01_Dyn_AnimBp
target_skeleton=/Game/Characters/Player/Common/Fortnite_Base_Head/Fortnite_Base_Head_Export_Skeleton
graphs [
 AnimGraph {nodes [
   n1 = Anim Dynamics - Bone: dyn_hair_01(Alpha=0.75) {
    angularBiasOverride=0.95
    angularSpringConstant=250
    bAngularSpring=true
    bGravityOverrideInSimSpace=true
    bLinearSpring=true
    bOverrideAngularBias=true
    gravityScale=0.125
    lODThreshold=1
    linearSpringConstant=350
    numSolverIterationsPostUpdate=4
    numSolverIterationsPreUpdate=16
    alphaBoolBlend {blendOption=Linear}
    boundBone {boneName=dyn_hair_01}
    chainEnd {boneName=dyn_hair_01_a}
    componentLinearAccScale=<0.8 0.8 0.8>
    physicsBodyDefinitions [{
      sphereCollisionRadius=0
      boundBone {boneName=dyn_hair_01}
      boxExtents=<5 5 5>
      constraintSetup {
       angularTargetAxis=AxisZ
       twistAxis=AxisZ
       angularLimitsMax=<0 0 0>
       angularLimitsMin=<0 0 0>
       angularTarget=<0 0 0>
       linearAxesMax=<0 0 2>
       linearAxesMin=<0 0 -2>
      }
      localJointOffset=<0 0 -0.1>
     }]
    planarLimits [{drivingBone {boneName=head}, planeTransform {location=<0 0 0>, rotation=<rot 0 0 0>}}]
   }
   n2 = Anim Dynamics - Bone: dyn_hair_02(ComponentPose=@n1, Alpha=0.75) {
    angularBiasOverride=0.95
    angularSpringConstant=250
    bAngularSpring=true
    bGravityOverrideInSimSpace=true
    bLinearSpring=true
    bOverrideAngularBias=true
    bUseSphericalLimits=true
    gravityScale=0.125
    lODThreshold=1
    linearSpringConstant=350
    numSolverIterationsPostUpdate=4
    numSolverIterationsPreUpdate=16
    alphaBoolBlend {blendOption=Linear}
    boundBone {boneName=dyn_hair_02}
    chainEnd {boneName=dyn_hair_02_a}
    componentLinearAccScale=<0.8 0.8 0.8>
    physicsBodyDefinitions [{
      sphereCollisionRadius=0
      boundBone {boneName=dyn_hair_02}
      boxExtents=<5 5 5>
      constraintSetup {
       angularTargetAxis=AxisZ
       twistAxis=AxisZ
       angularLimitsMax=<0 0 0>
       angularLimitsMin=<0 0 0>
       angularTarget=<0 0 0>
       linearAxesMax=<0 0 2>
       linearAxesMin=<0 0 -2>
      }
      localJointOffset=<0 0 -0.1>
     }]
    sphericalLimits [{limitRadius=5, drivingBone {boneName=head}, sphereLocalOffset=<0 0 0>}]
   }
   n3 = Anim Dynamics - Bone: dyn_hair_01_a(ComponentPose=@n2, Alpha=0.3, ExternalForce="0,0,0") {
    angularBiasOverride=0.95
    angularSpringConstant=150
    bAngularSpring=true
    bGravityOverrideInSimSpace=true
    bOverrideAngularBias=true
    gravityScale=0.6
    lODThreshold=1
    numSolverIterationsPostUpdate=4
    numSolverIterationsPreUpdate=16
    simulationSpace=Actor
    alphaBoolBlend {blendOption=Linear}
    boundBone {boneName=dyn_hair_01_a}
    physicsBodyDefinitions [{
      sphereCollisionRadius=0
      boundBone {boneName=dyn_hair_01_a}
      boxExtents=<5 5 5>
      constraintSetup {angularTargetAxis=AxisZ, twistAxis=AxisZ, angularLimitsMax=<89 30 0>, angularLimitsMin=<-89 -30 0>, angularTarget=<0 0 0>}
      localJointOffset=<0 0 -0.5>
     }]
   }
   Anim Dynamics - Bone: dyn_hair_02_a(ComponentPose=@n3, Alpha=0.25, ExternalForce="0,0,0") {
    angularBiasOverride=0.95
    angularSpringConstant=150
    bAngularSpring=true
    bGravityOverrideInSimSpace=true
    bOverrideAngularBias=true
    gravityScale=0.4
    lODThreshold=1
    numSolverIterationsPostUpdate=4
    numSolverIterationsPreUpdate=16
    simulationSpace=Actor
    alphaBoolBlend {blendOption=Linear}
    boundBone {boneName=dyn_hair_02_a}
    physicsBodyDefinitions [{
      sphereCollisionRadius=0
      boundBone {boneName=dyn_hair_02_a}
      boxExtents=<5 5 5>
      constraintSetup {angularTargetAxis=AxisZ, twistAxis=AxisZ, angularLimitsMax=<89 30 0>, angularLimitsMin=<-89 -30 0>, angularTarget=<0 0 0>}
      localJointOffset=<0 0 -0.5>
     }]
   }
   AnimGraphNode_SubInput_0 = Input Pose()
   n4 = Anim Dynamics - Bone: dyn_hair_01(ComponentPose=Local To Component(LocalPose=@AnimGraphNode_SubInput_0), Alpha=0.25) {
    angularBiasOverride=0.95
    angularSpringConstant=45
    bAngularSpring=true
    bGravityOverrideInSimSpace=true
    bOverrideAngularBias=true
    gravityScale=0.2
    lODThreshold=1
    numSolverIterationsPostUpdate=4
    numSolverIterationsPreUpdate=16
    windScale=0
    alphaBoolBlend {blendOption=Linear}
    boundBone {boneName=dyn_hair_01}
    physicsBodyDefinitions [{
      sphereCollisionRadius=0
      boundBone {boneName=dyn_hair_01}
      boxExtents=<5 5 5>
      constraintSetup {angularTargetAxis=AxisZ, twistAxis=AxisZ, angularLimitsMax=<0 12 0>, angularLimitsMin=<0 -10 0>, angularTarget=<0 0 0>}
      localJointOffset=<0 0 -0.6>
     }]
   }
   n5 = Anim Dynamics - Bone: dyn_hair_02(ComponentPose=@n4, Alpha=0.28) {
    angularBiasOverride=0.95
    angularSpringConstant=75
    bAngularSpring=true
    bGravityOverrideInSimSpace=true
    bOverrideAngularBias=true
    gravityScale=0.3
    lODThreshold=1
    numSolverIterationsPostUpdate=4
    numSolverIterationsPreUpdate=16
    windScale=0
    alphaBoolBlend {blendOption=Linear}
    boundBone {boneName=dyn_hair_02}
    physicsBodyDefinitions [{
      sphereCollisionRadius=0
      boundBone {boneName=dyn_hair_02}
      boxExtents=<5 5 5>
      constraintSetup {angularTargetAxis=AxisZ, twistAxis=AxisZ, angularLimitsMax=<0 10 0>, angularLimitsMin=<0 -15 0>, angularTarget=<0 0 0>}
      localJointOffset=<0 0 -0.4>
     }]
   }
   Output Pose(Result=Component To Local(ComponentPose=@n5)) {layerGroup=None}
  ]}
 EventGraph {nodes [Event BlueprintUpdateAnimation()]}
]
variables []
